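# Niagara trail effect for the Chug Jug consumable's cap, combining liquid drips, a ribbon and simulated fluid.
# Two sprite drip emitters inherit motion, and a ribbon follows skeletal mesh positions.
# A GPU Source emitter seeds a camera-facing 2D-grid PIC/FLIP fluid through particle attribute reads, surfaced with thickness-based colour and emissive.
# User.ColorVariant switches ribbon and fluid between blue-cyan and orange; drips and ribbon are distance-culled, fluid by significance.
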NiagaraSystem /Game/Effects/Fort_Effects/Effects/Characters/Generic/Consumables/NS_Consumable_CapTrail_ChugJug
dependencies=null
emitters {
 Drips {
  emitterUpdateScript [
   &EmitterState(Enable Distance Culling=true, MaxDistance=2000, Module.bUseMaxDistance=true, Module.bUseMinDistance=false, Scalability Mode=::NewEnumerator1)
   &SpawnRate(SpawnRate=35, Use Spawn Probability=false)
  ]
  particleSpawnScript [
   /Niagara/Modules/Spawn/Initialization/InitializeParticle(
    Lifetime=&UniformRangedFloat(Minimum=1, Maximum=2)
    Module.Write Color=true
    Module.Write Lifetime=true
    Module.Write Mass=true
    Module.Write Material Random=true
    Module.Write Position=false
    Module.Write Scale=false
    Module.Write SpriteRotation=false
    Module.Write SpriteSize=true
    Position=&SimulationPosition()
    Sprite Size=&Vector2DFromFloat(Value=&UniformRangedFloat(Minimum=0.25, Maximum=1.5))
   )
   &AddVelocity(Velocity=&UniformRangedVector(Minimum=<50 35 35>, Maximum=<25 -35 -35>))
   &InheritVelocity(Module.Limit Speed=true, Velocity Scale=<0.5 0.5 0.5>)
  ]
  particleUpdateScript [
   &ParticleState()
   &GravityForce(Gravity=<0 0 -750>)
   &Drag(
    Drag=&FloatFromCurve(FloatCurve=DataInterface() {propertyValues {Curve {keys [<at 0 0.5>,<at 0.8 1>]}}})
    Use Linear Drag=true
    Use Rotational Drag=false
   )
   &SolveForcesAndVelocity(Clamp Velocity=false, Limit Acceleration=false)
   &ScaleColor(Scale Alpha=&Lerp_Float(A=1, B=0, Alpha=$Particles.NormalizedAge), ScaleA=true, ScaleRGB=true)
   &ScaleSpriteSize(Scale Factor=&Vector2DFromFloat(Value=&FloatFromCurve(FloatCurve=DataInterface() {propertyValues {Curve {keys [<at 0.9 1>,<at 1 0>]}}})))
   &ScaleSpriteSizeBySpeed(Max Scale Factor=<2 25>)
  ]
  rendererClasses [&NiagaraSpriteRendererProperties]
  renderers [
   {rendererClass=&NiagaraSpriteRendererProperties}
   {rendererClass=&NiagaraSpriteRendererProperties, rendererIndex=1}
  ]
 }
 Drips001 {
  emitterUpdateScript [
   &EmitterState(Enable Distance Culling=true, MaxDistance=2000, Module.bUseMaxDistance=true, Module.bUseMinDistance=false, Scalability Mode=::NewEnumerator1)
   &SpawnRate(SpawnRate=22, Use Spawn Probability=false)
  ]
  particleSpawnScript [
   &InitializeParticle(
    Lifetime=&UniformRangedFloat(Minimum=1, Maximum=2)
    Module.Write Color=true
    Module.Write Lifetime=true
    Module.Write Mass=true
    Module.Write Material Random=true
    Module.Write Position=false
    Module.Write Scale=false
    Module.Write SpriteRotation=false
    Module.Write SpriteSize=true
    Position=&SimulationPosition()
    Sprite Size=&Vector2DFromFloat(Value=&UniformRangedFloat(Minimum=0.2, Maximum=0.885))
   )
   &AddVelocity(Velocity=&UniformRangedVector(Minimum=<50 35 35>, Maximum=<25 -35 -35>))
   &InheritVelocity(Module.Limit Speed=true, Velocity Scale=<0.4 0.4 0.4>)
  ]
  particleUpdateScript [
   &ParticleState()
   &GravityForce()
   &Drag(
    Drag=&FloatFromCurve(FloatCurve=DataInterface() {propertyValues {Curve {keys [<at 0 0.5>,<at 0.8 1>]}}})
    Use Linear Drag=true
    Use Rotational Drag=false
   )
   &SolveForcesAndVelocity(Clamp Velocity=false, Limit Acceleration=false)
   &ScaleColor(Scale Alpha=&Lerp_Float(A=1, B=0, Alpha=$Particles.NormalizedAge), ScaleA=true, ScaleRGB=true)
   &ScaleSpriteSize(Scale Factor=&Vector2DFromFloat(Value=&FloatFromCurve(FloatCurve=DataInterface() {propertyValues {Curve {keys [<at 0.9 1>,<at 1 0>]}}})))
   &ScaleSpriteSizeBySpeed(Max Scale Factor=<2 25>)
  ]
  rendererClasses [&NiagaraSpriteRendererProperties]
  renderers [
   {rendererClass=&NiagaraSpriteRendererProperties}
   {rendererClass=&NiagaraSpriteRendererProperties, rendererIndex=1}
  ]
 }
 FN_AttachedRibbon {
  emitterUpdateScript [
   &EmitterState(
    Enable Distance Culling=true
    Life Cycle Mode=::NewEnumerator1
    Loop Behavior=::NewEnumerator1
    Loop Delay=0.25
    Loop Duration Mode=::NewEnumerator1
    MaxDistance=3500
    Module.bUseMaxDistance=true
    Module.bUseMinDistance=false
    Scalability Mode=::NewEnumerator1
    UseLoopDelay=true
   )
   &FN_SpawnRibbon_v2(Spawn Mode=::NewEnumerator0)
  ]
  particleSpawnScript [&FN_InitializeRibbon_v3(
    Color:LinearColor=&SelectLinearColorFromArray(
     Array Sampling Mode=::NewEnumerator1
     Direct Array Index=$User.ColorVariant
     Color Selection Array=DataInterface() {propertyValues {ColorData [<rgba 0.1354 0.5929 1 1>, <rgba 0.6823 0.4709 0.0482 1>]}}
     Random Seed=?
    )
    Enable Skeletal Position=true
   )]
  particleUpdateScript [
   &ParticleState()
   &SolveForcesAndVelocity(Clamp Velocity=false, Limit Acceleration=false)
   &FN_UpdateRibbon(
    Enable Skeletal Position=true
    Initial Color:LinearColor=$Particles.Initial.Color
    Initial RibbonWidth:NiagaraFloat=$Particles.Initial.RibbonWidth
   )
   &ScaleSpriteSize(Scale Factor=&Vector2DFromCurve(Vector2Curve=DataInterface() {propertyValues {XCurve {keys [<at 0 1.25>,<at 1 2>]}, YCurve {keys [<at 0 1.25>,<at 1 1>]}}}, ScaleCurve=<1 1>))
   &ScaleColor(
    Scale RGB=&MakeVector(
     X=&MergeHelper_FloatSwitch(
      0=0
      1=&FloatFromCurve(FloatCurve=DataInterface() {propertyValues {Curve {keys [<at 0 1>,<at 1 0>]}}})
      2=?
      3=?
      Index=$User.ColorVariant
      Num=::NewEnumerator0
     )
     Y=&MergeHelper_FloatSwitch(
      0=&FloatFromCurve(FloatCurve=DataInterface() {propertyValues {Curve {keys [<at 0 1>,<at 1 0>]}}})
      1=&FloatFromCurve(FloatCurve=DataInterface() {propertyValues {Curve {keys [<at 0 1>,<at 1 0>]}}})
      2=?
      3=?
      Index=$User.ColorVariant
      Num=::NewEnumerator0
     )
     Z=&MergeHelper_FloatSwitch(
      0=&FloatFromCurve(FloatCurve=DataInterface() {propertyValues {Curve {keys [<at 0 1>,<at 1 0>]}}})
      1=&FloatFromCurve(FloatCurve=DataInterface() {propertyValues {Curve {keys [<at 0 1>,<at 1 0>]}}})
      2=?
      3=?
      Index=$User.ColorVariant
      Num=::NewEnumerator0
     )
    )
    ScaleA=true
    ScaleRGB=true
   )
  ]
  rendererClasses [&NiagaraRibbonRendererProperties]
  renderers [{rendererClass=&NiagaraRibbonRendererProperties}]
 }
 FN_PIC_FLIP_Fluid_ReadOtherEmitter_V1Behavior {
  simTarget=GPUComputeSim
  emitterSpawnScript [
   /SetVariables(
    Emitter.PressureGrid=DataInterface() {propertyValues {
      ClearBeforeNonIterationStage=true
      NumAttributes=0
      NumCellsMaxAxis=3
      NumCellsX=3
      NumCellsY=3
      OverrideBufferFormat=Float
      PreviewAttribute=None
      SetGridFromMaxAxis=false
      bOverrideFormat=true
      bPreviewGrid=false
      RenderTargetUserParameter {parameter {name=None, typeDefHandle {registeredTypeIndex=4}, varData []}}
      WorldBBoxSize=<100 100>
     }}
    Emitter.VelocityGrid=DataInterface() {propertyValues {
      ClearBeforeNonIterationStage=true
      NumAttributes=0
      NumCellsMaxAxis=3
      NumCellsX=128
      NumCellsY=64
      OverrideBufferFormat=Float
      PreviewAttribute=None
      SetGridFromMaxAxis=false
      bOverrideFormat=false
      bPreviewGrid=false
      RenderTargetUserParameter {parameter {name=None, typeDefHandle {registeredTypeIndex=4}, varData []}}
      WorldBBoxSize=<100 100>
     }}
   ) {isSetParameters=true}
   /NiagaraFluids/Modules/Grid2D/Grid2D_SetResolution(
    Grid:NiagaraDataInterfaceGrid2DCollection=$Emitter.VelocityGrid
    NumCellsMaxAxis=256
    Resolution Method:ENiagaraGrid2DResolution=::NewEnumerator1
    WorldGridExtents:Vector2f=<300 300>
   )
   &Grid2D_SetResolution(
    Grid:NiagaraDataInterfaceGrid2DCollection=$Emitter.PressureGrid
    OtherGrid:NiagaraDataInterfaceGrid2DCollection=$Emitter.VelocityGrid
    Resolution Method:ENiagaraGrid2DResolution=::NewEnumerator3
    WorldGridExtents:Vector2f=$Emitter.Grid2D_SetResolution.WorldGridExtents
   ) {module=Grid2D_SetResolution002}
   /SetVariables(Emitter.DeltaTime=0.033) {isSetParameters=true}
   /SetVariables(Emitter.SimRT=DataInterface() {propertyValues {
      MipMapGeneration=Disabled
      MipMapGenerationType=Linear
      OverrideRenderTargetFilter=TF_Default
      OverrideRenderTargetFormat=RTF_RGBA16f
      bInheritUserParameterSettings=false
      bOverrideFormat=true
      bPreviewRenderTarget=false
      RenderTargetUserParameter {parameter {name=None, typeDefHandle {registeredTypeIndex=4}, varData []}}
      Size=<0 0>
     }}) {isSetParameters=true}
   /NiagaraFluids/Modules/Grid2D/Grid2D_InitRT(Grid:NiagaraDataInterfaceGrid2DCollection=$Emitter.VelocityGrid, SimRT:NiagaraDataInterfaceRenderTarget2D=$Emitter.SimRT) {module=InitRTGrid2D}
   &self:FN_PIC_FLIP_Fluid_ReadOtherEmitter_V1Behavior.NiagaraScratchPadContainer_0.ScratchModule_0() {module=ScratchModule}
   &self:FN_PIC_FLIP_Fluid_ReadOtherEmitter_V1Behavior.NiagaraScratchPadContainer_0.ScratchModule_03_5(NumPressureIterations=20) {module=ScratchModule_03}
   /SetVariables() {isSetParameters=true}
   /SetVariables(Emitter.ThicknessRT=DataInterface() {propertyValues {
      MipMapGeneration=Disabled
      MipMapGenerationType=Linear
      OverrideRenderTargetFilter=TF_Default
      OverrideRenderTargetFormat=RTF_R16f
      bInheritUserParameterSettings=false
      bOverrideFormat=true
      bPreviewRenderTarget=false
      RenderTargetUserParameter {parameter {name=None, typeDefHandle {registeredTypeIndex=4}, varData []}}
      Size=<0 0>
     }}) {isSetParameters=true}
   &Grid2D_InitRT(Grid:NiagaraDataInterfaceGrid2DCollection=$Emitter.VelocityGrid, SimRT:NiagaraDataInterfaceRenderTarget2D=$Emitter.ThicknessRT)
   /SetVariables(
    Emitter.RasterizationGrid=DataInterface() {propertyValues {
      CellSize=1
      ClearBeforeNonIterationStage=true
      NumAttributes=1
      NumCellsMaxAxis=10
      Precision=1000
      ResetValue=100000000
      SetResolutionMethod=Independent
      NumCells=<3 3 3>
      WorldBBoxSize=<100 100 100>
     }}
    Emitter.RasterizationGridBack=DataInterface() {propertyValues {
      CellSize=1
      ClearBeforeNonIterationStage=true
      NumAttributes=1
      NumCellsMaxAxis=10
      Precision=1000
      ResetValue=0
      SetResolutionMethod=Independent
      NumCells=<3 3 3>
      WorldBBoxSize=<100 100 100>
     }}
   ) {isSetParameters=true}
   /NiagaraFluids/Modules/Grid3D/RasterizationGrid3D_SetResolution(
    Grid:NiagaraDataInterfaceRasterizationGrid3D=$Emitter.RasterizationGrid
    NumCellsX:NiagaraInt32=/Game/Effects/FlipSimulation/Niagara/DynamicInputs/GetGrid2DSize(Grid2D Collection:NiagaraDataInterfaceGrid2DCollection=$Emitter.VelocityGrid, Y:NiagaraBool=false)
    NumCellsY:NiagaraInt32=&GetGrid2DSize(Grid2D Collection:NiagaraDataInterfaceGrid2DCollection=$Emitter.VelocityGrid, Y:NiagaraBool=true)
    NumCellsZ=1
    WorldGridExtents:Vector3f=<1 1 1>
   )
   &RasterizationGrid3D_SetResolution(
    Grid:NiagaraDataInterfaceRasterizationGrid3D=$Emitter.RasterizationGridBack
    NumCellsX:NiagaraInt32=&GetGrid2DSize(Grid2D Collection:NiagaraDataInterfaceGrid2DCollection=$Emitter.VelocityGrid, Y:NiagaraBool=false)
    NumCellsY:NiagaraInt32=&GetGrid2DSize(Grid2D Collection:NiagaraDataInterfaceGrid2DCollection=$Emitter.VelocityGrid, Y:NiagaraBool=true)
    NumCellsZ=1
    WorldGridExtents:Vector3f=<1 1 1>
   ) {module=RasterizationGrid3D_SetResolution001}
  ]
  emitterUpdateScript [
   /NiagaraFN/Modules/Scalability/SignificantInstancesCull(Cull Response:ENiagaraExecutionStateManagement=::KillImmediately)
   &EmitterState()
   /NiagaraFluids/Modules/Grid2D/Grid2D_UpdateSimAttributes(Grid:NiagaraDataInterfaceGrid2DCollection=$Emitter.VelocityGrid, WorldGridSize:Vector2f=$Emitter.Grid2D_SetResolution.WorldGridExtents) {module=UpdateSimAttributes}
   /NiagaraFluids/Modules/Grid2D/Grid2D_CreateUnitToWorldTransform(
    Camera:NiagaraDataInterfaceCamera=$Emitter.CameraQuery
    Camera Facing=true
    Offset:NiagaraPosition=$Engine.Owner.Position
    Rotation:Quat4f=&self:FN_PIC_FLIP_Fluid_ReadOtherEmitter_V1Behavior.NiagaraScratchPadContainer_0.ScratchDynamicInput_01_1(Rotation:Quat4f=$Engine.Owner.Rotation)
    WorldGridExtents:Vector2f=$Emitter.Grid2D_SetResolution.WorldGridExtents
   )
   /SetVariables(Emitter.Position:NiagaraPosition=&ConvertVectorToPosition(
     Input Position=$Emitter.Grid2D_CreateUnitToWorldTransform.LocalToWorldTranslation
     Conversion Operation:ENiagara_LWCConvertVecToPos=::NewEnumerator0
    )) {isSetParameters=true}
   /NiagaraFluids/Modules/Grid2D/Grid2D_SetBounds(
    EmitterProperties=DataInterface() {propertyValues {EmitterBinding {bindingMode=Self, emitterName=None}}}
    LocalPivot:Vector2f=$Emitter.Grid2D_CreateUnitToWorldTransform.LocalPivot
    WorldGridExtents:Vector2f=$Emitter.Grid2D_SetResolution.WorldGridExtents
   )
   /SetVariables() {isSetParameters=true}
   /NiagaraFluids/Modules/Grid2D/SplitMatrixIntoRows(Matrix:NiagaraMatrix=$Emitter.Grid2D_CreateUnitToWorldTransform.WorldToUnit)
   /Game/Effects/FlipSimulation/Niagara/Modules/Flip_SurfacingParameters(
    BaseColor_HighThickness:LinearColor=&SelectLinearColorFromArray(
     Array Sampling Mode=::NewEnumerator1
     Direct Array Index=$User.ColorVariant
     Color Selection Array=DataInterface() {propertyValues {ColorData [<rgba 0 1 0.6667 1>, <rgba 1.5 0.625 0 1>]}}
     Random Seed=?
    )
    BaseColor_LowThickness:LinearColor=&SelectLinearColorFromArray(
     Array Sampling Mode=::NewEnumerator1
     Direct Array Index=$User.ColorVariant
     Color Selection Array=DataInterface() {propertyValues {ColorData [<rgba 0 1 1 1>, <rgba 1 0.45 0 1>]}}
     Random Seed=?
    )
    CustomBaseColor_HighThickness=true
    CustomBaseColor_lowThickness=true
    CustomDarkFaceBrightness=false
    CustomEdgeOffset=true
    CustomEnvReflection_Tint=true
    CustomFakeSSS_Color=true
    CustomHDRIReflection_Tint=false
    CustomHueShiftDarkFace=false
    CustomHueShiftLightFace=false
    CustomLightFaceBrightness=false
    CustomRefractionStrength=false
    CustomScaleBlurDistance=true
    CustomShadowStrength=false
    CustomThicknessDivideNumber=true
    CustomThicknessEmissive_HighThickness=true
    CustomThicknessEmissive_LowThickness=true
    CustomTotalOpacityScale=false
    EdgeOffset=-0.1
    EnvReflection_Tint:LinearColor=&SelectLinearColorFromArray(
     Array Sampling Mode=::NewEnumerator1
     Direct Array Index=$User.ColorVariant
     Color Selection Array=DataInterface() {propertyValues {ColorData [<rgba 0.1255 0.251 0.2275 1>, <rgba 0.251 0.2196 0.1255 1>]}}
     Random Seed=?
    )
    FakeSSSColor:LinearColor=&SelectLinearColorFromArray(
     Array Sampling Mode=::NewEnumerator1
     Direct Array Index=$User.ColorVariant
     Color Selection Array=DataInterface() {propertyValues {ColorData [<rgba 0 1.5 0.7529 1>, <rgba 1.5 0.625 0 1>]}}
     Random Seed=?
    )
    Preset:ENiagara_FlipRenderPresets=::NewEnumerator1
    ThicknessDivide=35
    ThicknessEmissive_HighThickness:LinearColor=&SelectLinearColorFromArray(
     Array Sampling Mode=::NewEnumerator1
     Direct Array Index=$User.ColorVariant
     Color Selection Array=DataInterface() {propertyValues {ColorData [<rgba 0 0.502 0.2078 1>, <rgba 0.502 0.2092 0 1>]}}
     Random Seed=?
    )
    ThicknessEmissive_LowThickness:LinearColor=&SelectLinearColorFromArray(
     Array Sampling Mode=::NewEnumerator1
     Direct Array Index=$User.ColorVariant
     Color Selection Array=DataInterface() {propertyValues {ColorData [<rgba 0 0.2 0.2 1>, <rgba 0.3 0.15 0 1>]}}
     Random Seed=?
    )
   )
   /SetVariables(
    Emitter.EnableSlurpRendererBlue:NiagaraBool=&Boolean_AndOperation(Bool1=$Emitter.FlipSurfacing.EnableSlurpRenderer, Bool2=$System.IsBlue)
    Emitter.EnableSlurpRendererOrange:NiagaraBool=&Boolean_AndOperation(Bool1=$Emitter.FlipSurfacing.EnableSlurpRenderer, Bool2=$System.IsOrange)
   ) {isSetParameters=true}
   &SpawnParticlesFromOtherEmitter(
    Attribute Reader=DataInterface() {propertyValues {EmitterBinding {bindingMode=Other, emitterName=Source}}}
    Spawn Rate=1500
    Spawn Rate Per Particle Cap=10000
    Use Maximum Age=false
    Use Minimum Age=false
   )
  ]
  particleSpawnScript [
   &InitializeParticle(
    Color=<rgba 0 0.3383 1 1>
    Lifetime=&RandomRangeFloat(Minimum=0.25, Maximum=0.3)
    Module.Write Color=false
    Module.Write Lifetime=true
    Module.Write Mass=false
    Module.Write Material Random=true
    Module.Write Position=false
    Module.Write Scale=false
    Module.Write SpriteRotation=false
    Module.Write SpriteSize=true
    Position=&SimulationPosition()
    Sprite Size=<4 4>
   )
   &SampleParticlesFromOtherEmitter(Applied Velocity Scale=2, Material Random Sampling:ENiagara_DirectReadSamplingMode=::NewEnumerator0, Velocity Sampling=::NewEnumerator0)
   /Niagara/Modules/Update/Forces/CurlNoiseForce(Noise Strength=0, Pan Noise=false)
   &LinearForce(Coordinate Space:ENiagaraCoordinateSpace=::World, Force=<0 0 100>)
   &ApplyInitialForces()
   &SphereLocation(
    MaskSpawnGroup=false
    Override Seed=false
    Sphere Radius=&Multiply_Float(
     A=&FloatFromCurve(FloatCurve=DataInterface() {propertyValues {Curve {keys [<at 0 0>,<at 1 1>]}}}, CurveIndex=$Engine.System.Age)
     B=1.5
    )
    Surface Only=false
    UseNonUniformScale=false
    UseOffset=false
    UseSphereOrientation=false
   )
   /SetVariables(Particles.Velocity=&Multiply_Vector(
     A=$Particles.Velocity
     B=&VectorFromFloat(Value=&ClampFloat(Float=&Power(a=&Divide_Float(A=0.0167, B=$Engine.DeltaTime), b=0.5), Min=0.5, Max=2))
    )) {isSetParameters=true}
  ]
  particleUpdateScript [
   &ParticleState()
   &GravityForce()
   &CurlNoiseForce(Noise Strength=400, Pan Noise=true, Pan Noise Field=<0.2 0.1 0.2>) {module=CurlNoiseForce001}
   /SetVariables(Transient.PhysicsForce:Vector3f=&Multiply_VectorByFloat(
     Vector=$Transient.PhysicsForce
     Float=&ClampFloat(
      Float=&Power(a=&Divide_Float(A=$Engine.DeltaTime, B=0.0167), b=0.5)
      Min=0.5
      Max=1.3
     )
    )) {isSetParameters=true}
   /SetVariables(Transient.StartPosition:NiagaraPosition=$Particles.Position) {isSetParameters=true}
   &SolveForcesAndVelocity(Clamp Velocity=false, DeltaTime:NiagaraFloat=$Emitter.DeltaTime, Limit Acceleration=false, Write to Presolve Properties=false)
   /SetVariables(Particles.Position:Vector3f=$Transient.StartPosition) {isSetParameters=true}
  ]
  rendererClasses [&NiagaraSpriteRendererProperties]
  renderers [
   {rendererClass=&NiagaraSpriteRendererProperties}
   {rendererClass=&NiagaraSpriteRendererProperties, rendererIndex=1}
   {rendererClass=&NiagaraSpriteRendererProperties, rendererIndex=2}
   {rendererClass=&NiagaraSpriteRendererProperties, rendererIndex=3}
   {rendererClass=&NiagaraSpriteRendererProperties, rendererIndex=4}
   {rendererClass=&NiagaraSpriteRendererProperties, rendererIndex=5}
  ]
 }
 Source {
  simTarget=GPUComputeSim
  emitterUpdateScript [
   &EmitterState()
   &SpawnBurst_Instantaneous(Use Spawn Probability=false)
   &SpawnRate(SpawnRate=130, Use Spawn Probability=false)
  ]
  particleSpawnScript [
   &InitializeParticle(
    Lifetime=&RandomRangeFloat(Minimum=0.1, Maximum=0.15)
    Mass=&RandomRangeFloat(Minimum=0.75, Maximum=1.75)
    Module.Write Color=true
    Module.Write Lifetime=true
    Module.Write Mass=true
    Module.Write Material Random=true
    Module.Write Position=false
    Module.Write Scale=false
    Module.Write SpriteRotation=false
    Module.Write SpriteSize=true
    Sprite Size=<1 2>
   )
   &AddVelocityInCone(
    Cone Angle=80
    Module.Use Velocity Falloff On Cone Axis=true
    Velocity Falloff Away From Cone Axis=0.333
    Velocity Strength=&RandomRangeFloat(Minimum=100, Maximum=100)
   ) {enabled=false}
   &SphereLocation(
    MaskSpawnGroup=false
    Offset=<0 0 45>
    Override Seed=false
    Sphere Coordinate Space=::World
    Sphere Radius=&FloatFromCurve(FloatCurve=DataInterface() {propertyValues {Curve {keys [<at 0 0.1>,<at 1 1>]}}}, CurveIndex=$Engine.System.Age, Scale Curve=3)
    Surface Only=false
    UseNonUniformScale=true
    UseOffset=true
    UseSphereOrientation=false
   ) {enabled=false}
   &InheritVelocity(Module.Limit Speed=true, Velocity Limit=1000, Velocity Scale=<0.1 0.1 0.1>) {enabled=false}
  ]
  particleUpdateScript [
   &ParticleState()
   &GravityForce(Gravity=<0 0 0>)
   &Drag(Drag=2, Use Linear Drag=true, Use Rotational Drag=false)
   &ScaleColor(
    Scale Alpha=&FloatFromCurve(FloatCurve=DataInterface() {propertyValues {Curve {keys [<at -0.0046 1.0113 RCIM_Cubic>,<at 1.0011 0.002 RCIM_Cubic>]}}})
    ScaleA=true
    ScaleRGB=true
   )
   &CurlNoiseForce(Noise Frequency=0.1, Noise Strength=300, Pan Noise=true, Pan Noise Field=<1 1 1>)
   &SolveForcesAndVelocity(Clamp Velocity=false, Limit Acceleration=false)
   &ScaleSpriteSizeBySpeed(
    Max Scale Factor=<1.2 3.3>
    Min Scale Factor=<0.75 0.3>
    Sample Scale Factor By Curve=true
    Scale Factor Curve=DataInterface() {propertyValues {Curve {keys [
        {arriveTangent=2.865, interpMode=RCIM_Cubic, time=0.1247, value=-0.0006}
        {arriveTangent=4.1522, interpMode=RCIM_Cubic, leaveTangent=4.1522, tangentMode=RCTM_User, time=0.3948, value=0.4893}
        <at 0.9958 0.9974 RCIM_Cubic>
       ]}}}
    Velocity Threshold=1500
   )
  ]
  rendererClasses [&NiagaraSpriteRendererProperties]
  renderers [{rendererClass=&NiagaraSpriteRendererProperties}]
 }
}
summary {
 systemName=NS_Consumable_CapTrail_ChugJug
 emitters [
  {emitterName=Drips, rendererClasses [&NiagaraSpriteRendererProperties]}
  {emitterName=Drips001, rendererClasses [&NiagaraSpriteRendererProperties]}
  {emitterName=FN_AttachedRibbon, rendererClasses [&NiagaraRibbonRendererProperties]}
  {bEnabled=false, emitterName=PIC_FLIP_Fluid, simTarget=GPUComputeSim, rendererClasses [&NiagaraSpriteRendererProperties]}
  {emitterName=Source, simTarget=GPUComputeSim, rendererClasses [&NiagaraSpriteRendererProperties]}
  {
   emitterName=FN_PIC_FLIP_Fluid_ReadOtherEmitter_V1Behavior
   simTarget=GPUComputeSim
   rendererClasses [&NiagaraSpriteRendererProperties]
  }
 ]
 userVariables [
  User.ColorVariant {type=NiagaraInt32}
  User.RibbonIDOffset {type=NiagaraInt32}
  User.SkeletalMesh {
   type=NiagaraDataInterfaceSkeletalMesh
   defaultValue {
    struct=NiagaraExt_VariableValue_DataInterface
    value {dataInterface=&self:NiagaraDataInterfaceSkeletalMesh_2}
   }
  }
 ]
}
