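# Niagara system for the Chug Jug consumable's liquid effect: gushing fluid, drips and splashes.
# A GPU Source emitter seeds a camera-facing 2D-grid PIC/FLIP fluid emitter, surfaced through render targets.
# CPU sprite emitters add blue and green drips plus large and mouth splashes, thinned or faded on selected platform quality levels.
# User.ColorVariant switches fluid and splash colors between a cyan-green and an orange palette.
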NiagaraSystem /Game/Effects/Fort_Effects/Effects/Characters/Generic/Consumables/NS_Consumable_ChugJug
dependencies {
 usedDataInterfaces [
  &NiagaraDataInterfacePlatformSet
  &NiagaraDataInterfaceCurve
  &NiagaraDataInterfaceColorCurve
  /Script/Niagara.NiagaraDataInterfaceGrid2DCollection
  /Script/Niagara.NiagaraDataInterfaceRenderTarget2D
  &NiagaraDataInterfaceCamera
  /Script/Niagara.NiagaraDataInterfaceRasterizationGrid3D
  /Script/Niagara.NiagaraDataInterfaceEmitterProperties
  &NiagaraDataInterfaceArrayColor
  &NiagaraDataInterfaceParticleRead
 ]
 usedDynamicInputs [&SetBoolByIntComparison]
 usedModules [
  &self:SystemSpawnScript.SystemScriptSource.SystemScriptGraph.NiagaraNodeAssignment_1.SetVariables_#1
  &SystemState
  &self:SystemSpawnScript.SystemScriptSource.SystemScriptGraph.NiagaraNodeAssignment_0.SetVariables_#2
 ]
}
emitters {
 DripsBlue {
  emitterUpdateScript [
   &EmitterState(
    Enable Distance Culling=true
    Life Cycle Mode=::NewEnumerator1
    Loop Duration=1
    MaxDistance=4000
    Module.bUseMaxDistance=true
    Module.bUseMinDistance=false
    Scalability Mode=::NewEnumerator1
    UseLoopDelay=false
   )
   &SpawnRate(
    SpawnRate=&FN_PickFloatDependOnPlatform(
     TargetPlatform=DataInterface() {propertyValues {Platforms {qualityLevelMask=-32, deviceProfileStates [{profileName=Windows, qualityLevelMask=24, setQualityLevelMask=24}]}}}
     TargetPlatformActive=30
     TargetPlatformNoActive=55
    )
    Use Spawn Probability=false
   )
  ]
  particleSpawnScript [
   /Niagara/Modules/Spawn/Initialization/InitializeParticle(
    Lifetime=&UniformRangedFloat(Minimum=1, Maximum=1.5)
    Module.Write Color=true
    Module.Write Lifetime=true
    Module.Write Mass=true
    Module.Write Material Random=true
    Module.Write Position=false
    Module.Write Scale=false
    Module.Write SpriteRotation=false
    Module.Write SpriteSize=true
    Position=&SimulationPosition()
    Sprite Size=&UniformRangedVector2D(Minimum=<0.5 0.5>, Maximum=<3 3>)
   )
   &FN_SetSpawnGroup_v2(Number of Spawn Groups=3)
   &SphereLocation(
    MaskSpawnGroup=false
    Override Seed=false
    Sphere Radius=2
    Surface Only=false
    UseNonUniformScale=false
    UseOffset=false
    UseSphereOrientation=false
   )
   &AddVelocity(Velocity=&MaskVectorBySpawnGroup(
     Masked Vector=&UniformRangedVector(Minimum=<200 10 10>, Maximum=<100 -10 -10>)
     Passthrough Vector=<0 0 0>
    ))
   &AddVelocity(Velocity=&MaskVectorBySpawnGroup(
     Masked Vector=&UniformRangedVector(Minimum=<150 -200 10>, Maximum=<150 -100 -10>)
     Passthrough Vector=<0 0 0>
     Spawn Group Mask=1
    )) {module=AddVelocity001}
   &AddVelocity(Velocity=&MaskVectorBySpawnGroup(
     Masked Vector=&UniformRangedVector(Minimum=<150 200 10>, Maximum=<150 100 -10>)
     Passthrough Vector=<1 1 1>
     Spawn Group Mask=2
    )) {module=AddVelocity002}
   &SystemLocation(MaskSpawnGroup=false, Offset=<0 0 41>)
  ]
  particleUpdateScript [
   &ParticleState()
   &GravityForce()
   &Drag(
    Drag=&FloatFromCurve(FloatCurve=DataInterface() {propertyValues {Curve {keys [<at 0 0.5>,<at 0.8 4>]}}})
    Use Linear Drag=true
    Use Rotational Drag=false
   )
   &SolveForcesAndVelocity(Clamp Velocity=false, Limit Acceleration=false)
   &ScaleSpriteSize(Scale Factor=&Vector2DFromFloat(Value=&FloatFromCurve(FloatCurve=DataInterface() {propertyValues {Curve {keys [<at 0.9 1>,<at 1 0>]}}})))
   &ScaleSpriteSizeBySpeed(Max Scale Factor=<2 16>)
   &ScaleColor(
    Scale Alpha=&FloatFromCurve(FloatCurve=DataInterface() {propertyValues {Curve {keys [
         {interpMode=RCIM_Cubic, tangentMode=RCTM_User, time=0, value=0}
         {interpMode=RCIM_Cubic, tangentMode=RCTM_User, time=0.15, value=1}
         {interpMode=RCIM_Cubic, tangentMode=RCTM_User, time=0.8, value=1}
         {arriveTangent=-11.4621, interpMode=RCIM_Cubic, leaveTangent=-11.4621, tangentMode=RCTM_User, time=1, value=0}
        ]}}})
    ScaleA=true
    ScaleRGB=false
   )
   &ScaleColor(
    Scale Alpha=&FN_PickFloatDependOnPlatform(
     TargetPlatform=DataInterface() {propertyValues {Platforms {qualityLevelMask=-32, deviceProfileStates [{profileName=Windows, qualityLevelMask=24, setQualityLevelMask=24}]}}}
     TargetPlatformActive=0.1
     TargetPlatformNoActive=1
    )
    ScaleA=true
    ScaleRGB=true
   ) {module=ScaleColor001}
  ]
  rendererClasses [&NiagaraSpriteRendererProperties]
  renderers [
   {rendererClass=&NiagaraSpriteRendererProperties}
   {rendererClass=&NiagaraSpriteRendererProperties, rendererIndex=1}
  ]
 }
 DripsGreen {
  emitterUpdateScript [
   &EmitterState(
    Enable Distance Culling=true
    Life Cycle Mode=::NewEnumerator1
    Loop Duration=1
    MaxDistance=4000
    Module.bUseMaxDistance=true
    Module.bUseMinDistance=false
    Scalability Mode=::NewEnumerator1
    UseLoopDelay=false
   )
   &SpawnRate(
    SpawnRate=&FN_PickFloatDependOnPlatform(
     TargetPlatform=DataInterface() {propertyValues {Platforms {qualityLevelMask=-32, deviceProfileStates [{profileName=Windows, qualityLevelMask=24, setQualityLevelMask=24}]}}}
     TargetPlatformActive=20
     TargetPlatformNoActive=75
    )
    Use Spawn Probability=false
   )
  ]
  particleSpawnScript [
   &InitializeParticle(
    Lifetime=&UniformRangedFloat(Minimum=1, Maximum=1.5)
    Module.Write Color=true
    Module.Write Lifetime=true
    Module.Write Mass=true
    Module.Write Material Random=true
    Module.Write Position=false
    Module.Write Scale=false
    Module.Write SpriteRotation=false
    Module.Write SpriteSize=true
    Position=&SimulationPosition()
    Sprite Size=&UniformRangedVector2D(Minimum=<0.5 0.5>, Maximum=<3 3>)
   )
   &FN_SetSpawnGroup_v2(Number of Spawn Groups=3)
   &SphereLocation(
    MaskSpawnGroup=false
    Override Seed=false
    Sphere Radius=2
    Surface Only=false
    UseNonUniformScale=false
    UseOffset=false
    UseSphereOrientation=false
   )
   &AddVelocity(Velocity=&MaskVectorBySpawnGroup(
     Masked Vector=&UniformRangedVector(Minimum=<200 10 10>, Maximum=<100 -10 -10>)
     Passthrough Vector=<0 0 0>
    ))
   &AddVelocity(Velocity=&MaskVectorBySpawnGroup(
     Masked Vector=&UniformRangedVector(Minimum=<150 -200 10>, Maximum=<150 -100 -10>)
     Passthrough Vector=<0 0 0>
     Spawn Group Mask=1
    )) {module=AddVelocity001}
   &AddVelocity(Velocity=&MaskVectorBySpawnGroup(
     Masked Vector=&UniformRangedVector(Minimum=<150 200 10>, Maximum=<150 100 -10>)
     Passthrough Vector=<1 1 1>
     Spawn Group Mask=2
    )) {module=AddVelocity002}
   &SystemLocation(MaskSpawnGroup=false, Offset=<0 0 41>)
  ]
  particleUpdateScript [
   &ParticleState()
   &GravityForce()
   &Drag(
    Drag=&FloatFromCurve(FloatCurve=DataInterface() {propertyValues {Curve {keys [<at 0 0.5>,<at 0.8 4>]}}})
    Use Linear Drag=true
    Use Rotational Drag=false
   )
   &SolveForcesAndVelocity(Clamp Velocity=false, Limit Acceleration=false)
   &ScaleSpriteSize(Scale Factor=&Vector2DFromFloat(Value=&FloatFromCurve(FloatCurve=DataInterface() {propertyValues {Curve {keys [<at 0.9 1>,<at 1 0>]}}})))
   &ScaleSpriteSizeBySpeed(Max Scale Factor=<2 16>)
   &ScaleColor(
    Scale Alpha=&FloatFromCurve(FloatCurve=DataInterface() {propertyValues {Curve {keys [
         {interpMode=RCIM_Cubic, tangentMode=RCTM_User, time=0, value=0}
         {interpMode=RCIM_Cubic, tangentMode=RCTM_User, time=0.15, value=1}
         {interpMode=RCIM_Cubic, tangentMode=RCTM_User, time=0.8, value=1}
         {arriveTangent=-11.4621, interpMode=RCIM_Cubic, leaveTangent=-11.4621, tangentMode=RCTM_User, time=1, value=0}
        ]}}})
    ScaleA=true
    ScaleRGB=false
   )
   &ScaleColor(
    Scale Alpha=&FN_PickFloatDependOnPlatform(
     TargetPlatform=DataInterface() {propertyValues {Platforms {qualityLevelMask=-32, deviceProfileStates [{profileName=Windows, qualityLevelMask=24, setQualityLevelMask=24}]}}}
     TargetPlatformActive=0.1
     TargetPlatformNoActive=1
    )
    ScaleA=true
    ScaleRGB=true
   ) {module=ScaleColor001}
  ]
  rendererClasses [&NiagaraSpriteRendererProperties]
  renderers [
   {rendererClass=&NiagaraSpriteRendererProperties}
   {rendererClass=&NiagaraSpriteRendererProperties, rendererIndex=1}
  ]
 }
 FN_PIC_FLIP_Fluid_ReadOtherEmitter_V1Behavior {
  simTarget=GPUComputeSim
  emitterSpawnScript [
   /SetVariables(
    Emitter.PressureGrid=DataInterface() {propertyValues {
      ClearBeforeNonIterationStage=true
      NumAttributes=0
      NumCellsMaxAxis=3
      NumCellsX=3
      NumCellsY=3
      OverrideBufferFormat=Float
      PreviewAttribute=None
      SetGridFromMaxAxis=false
      bOverrideFormat=true
      bPreviewGrid=false
      RenderTargetUserParameter {parameter {name=None, typeDefHandle {registeredTypeIndex=4}, varData []}}
      WorldBBoxSize=<100 100>
     }}
    Emitter.VelocityGrid=DataInterface() {propertyValues {
      ClearBeforeNonIterationStage=true
      NumAttributes=0
      NumCellsMaxAxis=3
      NumCellsX=128
      NumCellsY=64
      OverrideBufferFormat=Float
      PreviewAttribute=None
      SetGridFromMaxAxis=false
      bOverrideFormat=false
      bPreviewGrid=false
      RenderTargetUserParameter {parameter {name=None, typeDefHandle {registeredTypeIndex=4}, varData []}}
      WorldBBoxSize=<100 100>
     }}
   ) {isSetParameters=true}
   /NiagaraFluids/Modules/Grid2D/Grid2D_SetResolution(
    Grid:NiagaraDataInterfaceGrid2DCollection=$Emitter.VelocityGrid
    NumCellsMaxAxis=256
    Resolution Method:ENiagaraGrid2DResolution=::NewEnumerator1
    WorldGridExtents:Vector2f=<300 300>
   )
   &Grid2D_SetResolution(
    Grid:NiagaraDataInterfaceGrid2DCollection=$Emitter.PressureGrid
    OtherGrid:NiagaraDataInterfaceGrid2DCollection=$Emitter.VelocityGrid
    Resolution Method:ENiagaraGrid2DResolution=::NewEnumerator3
    WorldGridExtents:Vector2f=$Emitter.Grid2D_SetResolution.WorldGridExtents
   ) {module=Grid2D_SetResolution002}
   /SetVariables(Emitter.DeltaTime=0.0166) {isSetParameters=true}
   /SetVariables(Emitter.SimRT=DataInterface() {propertyValues {
      MipMapGeneration=Disabled
      MipMapGenerationType=Linear
      OverrideRenderTargetFilter=TF_Default
      OverrideRenderTargetFormat=RTF_RGBA16f
      bInheritUserParameterSettings=false
      bOverrideFormat=true
      bPreviewRenderTarget=false
      RenderTargetUserParameter {parameter {name=None, typeDefHandle {registeredTypeIndex=4}, varData []}}
      Size=<0 0>
     }}) {isSetParameters=true}
   /NiagaraFluids/Modules/Grid2D/Grid2D_InitRT(Grid:NiagaraDataInterfaceGrid2DCollection=$Emitter.VelocityGrid, SimRT:NiagaraDataInterfaceRenderTarget2D=$Emitter.SimRT) {module=InitRTGrid2D}
   &self:FN_PIC_FLIP_Fluid_ReadOtherEmitter_V1Behavior.NiagaraScratchPadContainer_0.ScratchModule_0() {module=ScratchModule}
   &self:FN_PIC_FLIP_Fluid_ReadOtherEmitter_V1Behavior.NiagaraScratchPadContainer_0.ScratchModule_03_5(NumPressureIterations=10) {module=ScratchModule_03}
   /SetVariables() {isSetParameters=true}
   /SetVariables(Emitter.ThicknessRT=DataInterface() {propertyValues {
      MipMapGeneration=Disabled
      MipMapGenerationType=Linear
      OverrideRenderTargetFilter=TF_Default
      OverrideRenderTargetFormat=RTF_R16f
      bInheritUserParameterSettings=false
      bOverrideFormat=true
      bPreviewRenderTarget=false
      RenderTargetUserParameter {parameter {name=None, typeDefHandle {registeredTypeIndex=4}, varData []}}
      Size=<0 0>
     }}) {isSetParameters=true}
   &Grid2D_InitRT(Grid:NiagaraDataInterfaceGrid2DCollection=$Emitter.VelocityGrid, SimRT:NiagaraDataInterfaceRenderTarget2D=$Emitter.ThicknessRT)
   /SetVariables(
    Emitter.RasterizationGrid=DataInterface() {propertyValues {
      CellSize=1
      ClearBeforeNonIterationStage=true
      NumAttributes=1
      NumCellsMaxAxis=10
      Precision=1000
      ResetValue=100000000
      SetResolutionMethod=Independent
      NumCells=<3 3 3>
      WorldBBoxSize=<100 100 100>
     }}
    Emitter.RasterizationGridBack=DataInterface() {propertyValues {
      CellSize=1
      ClearBeforeNonIterationStage=true
      NumAttributes=1
      NumCellsMaxAxis=10
      Precision=1000
      ResetValue=0
      SetResolutionMethod=Independent
      NumCells=<3 3 3>
      WorldBBoxSize=<100 100 100>
     }}
   ) {isSetParameters=true}
   /NiagaraFluids/Modules/Grid3D/RasterizationGrid3D_SetResolution(
    Grid:NiagaraDataInterfaceRasterizationGrid3D=$Emitter.RasterizationGrid
    NumCellsX:NiagaraInt32=/Game/Effects/FlipSimulation/Niagara/DynamicInputs/GetGrid2DSize(Grid2D Collection:NiagaraDataInterfaceGrid2DCollection=$Emitter.VelocityGrid, Y:NiagaraBool=false)
    NumCellsY:NiagaraInt32=&GetGrid2DSize(Grid2D Collection:NiagaraDataInterfaceGrid2DCollection=$Emitter.VelocityGrid, Y:NiagaraBool=true)
    NumCellsZ=1
    WorldGridExtents:Vector3f=<1 1 1>
   )
   &RasterizationGrid3D_SetResolution(
    Grid:NiagaraDataInterfaceRasterizationGrid3D=$Emitter.RasterizationGridBack
    NumCellsX:NiagaraInt32=&GetGrid2DSize(Grid2D Collection:NiagaraDataInterfaceGrid2DCollection=$Emitter.VelocityGrid, Y:NiagaraBool=false)
    NumCellsY:NiagaraInt32=&GetGrid2DSize(Grid2D Collection:NiagaraDataInterfaceGrid2DCollection=$Emitter.VelocityGrid, Y:NiagaraBool=true)
    NumCellsZ=1
    WorldGridExtents:Vector3f=<1 1 1>
   ) {module=RasterizationGrid3D_SetResolution001}
  ]
  emitterUpdateScript [
   /NiagaraFN/Modules/Scalability/SignificantInstancesCull(Cull Response:ENiagaraExecutionStateManagement=::KillImmediately)
   &EmitterState()
   /NiagaraFluids/Modules/Grid2D/Grid2D_UpdateSimAttributes(Grid:NiagaraDataInterfaceGrid2DCollection=$Emitter.VelocityGrid, WorldGridSize:Vector2f=$Emitter.Grid2D_SetResolution.WorldGridExtents) {module=UpdateSimAttributes}
   /NiagaraFluids/Modules/Grid2D/Grid2D_CreateUnitToWorldTransform(
    Camera:NiagaraDataInterfaceCamera=$Emitter.CameraQuery
    Camera Facing=true
    Local Pivot:Vector2f=<0.5 0.65>
    Offset:NiagaraPosition=$Engine.Owner.Position
    Rotation:Quat4f=&self:FN_PIC_FLIP_Fluid_ReadOtherEmitter_V1Behavior.NiagaraScratchPadContainer_0.ScratchDynamicInput_01_1(Rotation:Quat4f=$Engine.Owner.Rotation)
    WorldGridExtents:Vector2f=$Emitter.Grid2D_SetResolution.WorldGridExtents
   )
   /SetVariables(Emitter.Position:NiagaraPosition=&ConvertVectorToPosition(
     Input Position=$Emitter.Grid2D_CreateUnitToWorldTransform.LocalToWorldTranslation
     Conversion Operation:ENiagara_LWCConvertVecToPos=::NewEnumerator0
    )) {isSetParameters=true}
   /NiagaraFluids/Modules/Grid2D/Grid2D_SetBounds(
    EmitterProperties=DataInterface() {propertyValues {EmitterBinding {bindingMode=Self, emitterName=None}}}
    LocalPivot:Vector2f=$Emitter.Grid2D_CreateUnitToWorldTransform.LocalPivot
    WorldGridExtents:Vector2f=$Emitter.Grid2D_SetResolution.WorldGridExtents
   )
   /SetVariables(Emitter.CameraDirectionDotZ:NiagaraFloat=&Abs_Float(Float:NiagaraFloat=&DotProduct(
      Vector A=&ReturnCameraVector(
       Camera Vectors=::NewEnumerator1
       Coordinate Space=::Simulation
       CameraQuery=DataInterface() {propertyValues {bRequireCurrentFrameData=false}}
      )
      Vector B=<0 0 1>
     ))) {isSetParameters=true}
   /SetVariables(Emitter.OrthogonalVelocityMult:NiagaraFloat=&FN_LOD_Distance_Remap(
     Start Distance=400
     End Distance=500
     Minimum Output=&FloatFromCurve(
      FloatCurve=DataInterface() {propertyValues {Curve {keys [
          {interpMode=RCIM_Cubic, leaveTangent=1.0344, tangentMode=RCTM_Break, time=0.5, value=0}
          {arriveTangent=3.0974, interpMode=RCIM_Cubic, tangentMode=RCTM_Break, time=1, value=1}
         ]}}}
      CurveIndex=$Emitter.CameraDirectionDotZ
     )
     Maximum Output=0.93
     Clamp Output=true
    )) {isSetParameters=true}
   /SetVariables(Emitter.CurlNoiseMult:NiagaraFloat=&FloatFromCurve(
     FloatCurve=DataInterface() {propertyValues {Curve {keys [
         {interpMode=RCIM_Cubic, leaveTangent=-0.984, tangentMode=RCTM_Break, time=0.5, value=1}
         {arriveTangent=-0.195, interpMode=RCIM_Cubic, tangentMode=RCTM_Break, time=1, value=0}
        ]}}}
     CurveIndex=$Emitter.CameraDirectionDotZ
    )) {isSetParameters=true}
   /NiagaraFluids/Modules/Grid2D/SplitMatrixIntoRows(Matrix:NiagaraMatrix=$Emitter.Grid2D_CreateUnitToWorldTransform.WorldToUnit)
   /Game/Effects/FlipSimulation/Niagara/Modules/Flip_SurfacingParameters(
    BaseColor_HighThickness:LinearColor=&SelectLinearColorFromArray(
     Array Sampling Mode=::NewEnumerator1
     Direct Array Index=$User.ColorVariant
     Color Selection Array=DataInterface() {propertyValues {ColorData [<rgba 0 1 0.6667 1>, <rgba 1.5 0.625 0 1>]}}
     Random Seed=?
    )
    BaseColor_LowThickness:LinearColor=&SelectLinearColorFromArray(
     Array Sampling Mode=::NewEnumerator1
     Direct Array Index=$User.ColorVariant
     Color Selection Array=DataInterface() {propertyValues {ColorData [<rgba 0 1 1 1>, <rgba 1 0.45 0 1>]}}
     Random Seed=?
    )
    CustomBaseColor_HighThickness=true
    CustomBaseColor_lowThickness=true
    CustomDarkFaceBrightness=false
    CustomEdgeOffset=true
    CustomEnvReflection_Tint=true
    CustomFakeSSS_Color=true
    CustomHDRIReflection_Tint=false
    CustomHueShiftDarkFace=false
    CustomHueShiftLightFace=false
    CustomLightFaceBrightness=false
    CustomRefractionStrength=false
    CustomScaleBlurDistance=true
    CustomShadowStrength=false
    CustomThicknessDivideNumber=true
    CustomThicknessEmissive_HighThickness=true
    CustomThicknessEmissive_LowThickness=true
    CustomTotalOpacityScale=false
    EdgeOffset=0.05
    EnvReflection_Tint:LinearColor=&SelectLinearColorFromArray(
     Array Sampling Mode=::NewEnumerator1
     Direct Array Index=$User.ColorVariant
     Color Selection Array=DataInterface() {propertyValues {ColorData [<rgba 0.1255 0.251 0.2275 1>, <rgba 0.251 0.2196 0.1255 1>]}}
     Random Seed=?
    )
    FakeSSSColor:LinearColor=&SelectLinearColorFromArray(
     Array Sampling Mode=::NewEnumerator1
     Direct Array Index=$User.ColorVariant
     Color Selection Array=DataInterface() {propertyValues {ColorData [<rgba 0 1.5 0.7529 1>, <rgba 1.5 0.625 0 1>]}}
     Random Seed=?
    )
    Preset:ENiagara_FlipRenderPresets=::NewEnumerator1
    ScaleBlurDistance=1.5
    ThicknessDivide=80
    ThicknessEmissive_HighThickness:LinearColor=&SelectLinearColorFromArray(
     Array Sampling Mode=::NewEnumerator1
     Direct Array Index=$User.ColorVariant
     Color Selection Array=DataInterface() {propertyValues {ColorData [<rgba 0 0.502 0.2078 1>, <rgba 0.502 0.2092 0 1>]}}
     Random Seed=?
    )
    ThicknessEmissive_LowThickness:LinearColor=&SelectLinearColorFromArray(
     Array Sampling Mode=::NewEnumerator1
     Direct Array Index=$User.ColorVariant
     Color Selection Array=DataInterface() {propertyValues {ColorData [<rgba 0 0.2 0.2 1>, <rgba 0.3 0.15 0 1>]}}
     Random Seed=?
    )
   )
   /SetVariables(
    Emitter.EnableSlurpRendererBlue:NiagaraBool=&Boolean_AndOperation(Bool1=$Emitter.FlipSurfacing.EnableSlurpRenderer, Bool2=$System.IsBlue)
    Emitter.EnableSlurpRendererOrange:NiagaraBool=&Boolean_AndOperation(Bool1=$Emitter.FlipSurfacing.EnableSlurpRenderer, Bool2=$System.IsOrange)
   ) {isSetParameters=true}
   &SpawnParticlesFromOtherEmitter(
    Attribute Reader=DataInterface() {propertyValues {EmitterBinding {bindingMode=Other, emitterName=Source}}}
    Spawn Rate=10000
    Spawn Rate Per Particle Cap=10000
    Use Maximum Age=false
    Use Minimum Age=false
   )
  ]
  particleSpawnScript [
   &InitializeParticle(
    Color=<rgba 0 0.3383 1 1>
    Lifetime=&RandomRangeFloat(Minimum=1.1, Maximum=1.2)
    Module.Write Color=false
    Module.Write Lifetime=true
    Module.Write Mass=false
    Module.Write Material Random=true
    Module.Write Position=false
    Module.Write Scale=false
    Module.Write SpriteRotation=false
    Module.Write SpriteSize=true
    Position=&SimulationPosition()
    Sprite Size=<4 4>
   )
   &SampleParticlesFromOtherEmitter(Applied Velocity Scale=2, Material Random Sampling:ENiagara_DirectReadSamplingMode=::NewEnumerator0, Velocity Sampling=::NewEnumerator0)
   /Niagara/Modules/Update/Forces/CurlNoiseForce(Noise Frequency=0.03, Noise Strength=135, Pan Noise=true, Pan Noise Field=<0 0 2>)
   &ApplyInitialForces()
   &SphereLocation(
    MaskSpawnGroup=false
    Override Seed=false
    Sphere Radius=&Multiply_Float(
     A=&FloatFromCurve(FloatCurve=DataInterface() {propertyValues {Curve {keys [<at 0 0>,<at 1 1>]}}}, CurveIndex=$Engine.System.Age)
     B=1
    )
    Surface Only=false
    UseNonUniformScale=false
    UseOffset=false
    UseSphereOrientation=false
   )
   /SetVariables(Particles.Velocity=&Multiply_Vector(
     A=$Particles.Velocity
     B=&VectorFromFloat(Value=&ClampFloat(Float=&Power(a=&Divide_Float(A=0.0167, B=$Engine.DeltaTime), b=0.5), Min=1, Max=2))
    )) {isSetParameters=true}
  ]
  particleUpdateScript [
   &ParticleState()
   &GravityForce(Gravity=<0 0 -1250>)
   &CurlNoiseForce(Noise Strength=&Multiply_Float(A=$Emitter.CurlNoiseMult, B=600), Pan Noise=true, Pan Noise Field=<0.2 0.1 1>) {module=CurlNoiseForce001}
   /SetVariables(Transient.PhysicsForce:Vector3f=&Multiply_VectorByFloat(
     Vector=$Transient.PhysicsForce
     Float=&ClampFloat(Float=&Power(a=&Divide_Float(A=$Engine.DeltaTime, B=0.0167), b=0.5), Min=1, Max=1.5)
    )) {isSetParameters=true}
   /SetVariables(Transient.StartPosition:NiagaraPosition=$Particles.Position) {isSetParameters=true}
   &SolveForcesAndVelocity(Clamp Velocity=false, DeltaTime:NiagaraFloat=$Emitter.DeltaTime, Limit Acceleration=false, Write to Presolve Properties=false)
   /SetVariables(Particles.Position:Vector3f=$Transient.StartPosition) {isSetParameters=true}
  ]
  rendererClasses [&NiagaraSpriteRendererProperties]
  renderers [
   {rendererClass=&NiagaraSpriteRendererProperties}
   {rendererClass=&NiagaraSpriteRendererProperties, rendererIndex=1}
   {rendererClass=&NiagaraSpriteRendererProperties, rendererIndex=2}
   {rendererClass=&NiagaraSpriteRendererProperties, rendererIndex=3}
   {rendererClass=&NiagaraSpriteRendererProperties, rendererIndex=4}
   {rendererClass=&NiagaraSpriteRendererProperties, rendererIndex=5}
  ]
 }
 Source {
  simTarget=GPUComputeSim
  emitterUpdateScript [
   &SignificantInstancesCull(
    Cull Response:ENiagaraExecutionStateManagement=::SleepAndLetParticlesFinish
    First Frame Only=false
    Num Instances:NiagaraInt32=$System.NumFluidInstances
   )
   &EmitterState(Life Cycle Mode=::NewEnumerator1, Loop Duration=3, Scalability Mode=::NewEnumerator1, UseLoopDelay=false)
   &SpawnBurst_Instantaneous(Use Spawn Probability=false)
   &SpawnRate(SpawnRate=35, Use Spawn Probability=false)
  ]
  particleSpawnScript [
   &InitializeParticle(
    Lifetime=&RandomRangeFloat(Minimum=0.2, Maximum=0.3)
    Mass=&RandomRangeFloat(Minimum=0.75, Maximum=1.75)
    Module.Write Color=true
    Module.Write Lifetime=true
    Module.Write Mass=true
    Module.Write Material Random=true
    Module.Write Position=false
    Module.Write Scale=false
    Module.Write SpriteRotation=false
    Module.Write SpriteSize=true
    Sprite Size=<1 2>
   )
   &AddVelocityInCone(
    Cone Angle=80
    Cone Axis=<1 0 -1>
    Module.Use Velocity Falloff On Cone Axis=true
    Velocity Falloff Away From Cone Axis=0.333
    Velocity Strength=&RandomRangeFloat(Minimum=100, Maximum=100)
   )
   &SphereLocation(
    MaskSpawnGroup=false
    Offset=<-13 0 45>
    Override Seed=false
    Sphere Coordinate Space=::World
    Sphere Radius=&FloatFromCurve(FloatCurve=DataInterface() {propertyValues {Curve {keys [<at 0 0.1>,<at 1 1>]}}}, CurveIndex=$Engine.System.Age, Scale Curve=5)
    Surface Only=false
    UseNonUniformScale=true
    UseOffset=true
    UseSphereOrientation=false
   )
   &InheritVelocity(Module.Limit Speed=true, Velocity Limit=1000, Velocity Scale=<0.1 0.1 0.1>)
  ]
  particleUpdateScript [
   &ParticleState()
   &GravityForce(Gravity=<0 0 0>)
   &Drag(Drag=2, Use Linear Drag=true, Use Rotational Drag=false)
   &ScaleColor(
    Scale Alpha=&FloatFromCurve(FloatCurve=DataInterface() {propertyValues {Curve {keys [<at -0.0046 1.0113 RCIM_Cubic>,<at 1.0011 0.002 RCIM_Cubic>]}}})
    ScaleA=true
    ScaleRGB=true
   )
   &CurlNoiseForce(Noise Frequency=0.1, Noise Strength=300, Pan Noise=true, Pan Noise Field=<1 1 1>)
   &SolveForcesAndVelocity(Clamp Velocity=false, Limit Acceleration=false)
   &ScaleSpriteSizeBySpeed(
    Max Scale Factor=<1.2 3.3>
    Min Scale Factor=<0.75 0.3>
    Sample Scale Factor By Curve=true
    Scale Factor Curve=DataInterface() {propertyValues {Curve {keys [
        {arriveTangent=2.865, interpMode=RCIM_Cubic, time=0.1247, value=-0.0006}
        {arriveTangent=4.1522, interpMode=RCIM_Cubic, leaveTangent=4.1522, tangentMode=RCTM_User, time=0.3948, value=0.4893}
        <at 0.9958 0.9974 RCIM_Cubic>
       ]}}}
    Velocity Threshold=1500
   )
  ]
  rendererClasses [&NiagaraSpriteRendererProperties]
  renderers [{rendererClass=&NiagaraSpriteRendererProperties}]
 }
 SpriteSplashesBig {
  emitterUpdateScript [
   &EmitterState(
    Enable Distance Culling=true
    Life Cycle Mode=::NewEnumerator1
    Loop Duration=1
    MaxDistance=10000
    Module.bUseMaxDistance=true
    Module.bUseMinDistance=false
    Scalability Mode=::NewEnumerator1
    UseLoopDelay=false
   )
   &SpawnRate(
    SpawnRate=&Multiply_Float(
     A=&Sine(Normalized Angle=$Emitter.Age, Period=1, Scale=24, Bias=12)
     B=&FN_PickFloatDependOnPlatform(
      TargetPlatform=DataInterface() {propertyValues {Platforms {qualityLevelMask=-32, deviceProfileStates [{profileName=Windows, qualityLevelMask=24, setQualityLevelMask=24}]}}}
      TargetPlatformActive=0.3
      TargetPlatformNoActive=1
     )
    )
    Use Spawn Probability=false
   )
  ]
  particleSpawnScript [
   &InitializeParticle(
    Color=&MakeLinearColorFromVectorAndFloat(
     "Vector (RGB)"=<10 10 10>
     "Float (Alpha)"=&Lerp_Float(
      A=0.15
      B=0.5
      Alpha=&FloatFromCurve(
       FloatCurve=DataInterface() {propertyValues {Curve {keys [
           {interpMode=RCIM_Cubic, tangentMode=RCTM_User, time=0, value=0}
           {arriveTangent=0.8887, interpMode=RCIM_Cubic, leaveTangent=0.8887, tangentMode=RCTM_User, time=0.8, value=0.2}
           {arriveTangent=6.4377, interpMode=RCIM_Cubic, leaveTangent=6.4377, tangentMode=RCTM_User, time=1, value=1}
          ]}}}
       CurveIndex=&RandomRangeFloat(Minimum=0, Maximum=1)
      )
     )
    )
    Lifetime=&UniformRangedFloat(Minimum=0.5, Maximum=0.75)
    Module.Write Color=true
    Module.Write Lifetime=true
    Module.Write Mass=true
    Module.Write Material Random=true
    Module.Write Position=false
    Module.Write Scale=false
    Module.Write SpriteRotation=true
    Module.Write SpriteSize=true
    Position=&SimulationPosition()
    Sprite Rotation=&UniformRangedFloat(Minimum=0, Maximum=360)
    Sprite Size=&UniformRangedVector2D(Minimum=<30 30>, Maximum=<50 50>)
   )
   &SphereLocation(
    MaskSpawnGroup=false
    Override Seed=false
    Sphere Radius=2
    Surface Only=false
    UseNonUniformScale=false
    UseOffset=false
    UseSphereOrientation=false
   )
   &AddVelocity(Velocity=&Multiply_Vector(
     A=$Output.SphereLocation.SphereNormal
     B=&VectorFromFloat(Value=&UniformRangedFloat(Minimum=162.5, Maximum=240))
    )) {module=AddVelocity001}
   /SetVariables(Particles.SubImageIndex=&UniformRangedFloat(Minimum=0, Maximum=3)) {isSetParameters=true, module=SetVariables_#3}
   &AddVelocity(Velocity=&UniformRangedVector(Minimum=<-10 -10 100>, Maximum=<10 10 200>))
   &SystemLocation(MaskSpawnGroup=false, Offset=<0 0 41>)
  ]
  particleUpdateScript [
   &ParticleState()
   &AccelerationForce(Acceleration=<0 0 -980>)
   &Drag(
    Drag=&FloatFromCurve(FloatCurve=DataInterface() {propertyValues {Curve {keys [<at 0 0.5>,<at 0.8 4>]}}})
    Use Linear Drag=true
    Use Rotational Drag=false
   )
   &SpriteRotationRate(Rotation Rate=&Multiply_Float(
     A=&Lerp_Float(A=-1, B=1, Alpha=$Particles.Initial.MaterialRandom)
     B=&Lerp_Float(A=125, B=0, Alpha=&Power(a=$Particles.NormalizedAge, b=5))
    ))
   &SolveForcesAndVelocity(Clamp Velocity=false, Limit Acceleration=false)
   &ScaleColor(
    Scale RGB=&MergeHelper_VectorSwitch(
     0=&MakeVectorFromLinearColorRGB(Color=&ColorFromCurve(ColorCurve=DataInterface() {propertyValues {
         AlphaCurve {keys [<at 0 1 RCIM_Cubic>]}
         BlueCurve {keys [<at 0 0.3283 RCIM_Cubic>, {arriveTangent=2.5158, interpMode=RCIM_Cubic, leaveTangent=2.5158, tangentMode=RCTM_User, time=0.5, value=1}]}
         GreenCurve {keys [<at 0 1 RCIM_Cubic>, {arriveTangent=-1.8785, interpMode=RCIM_Cubic, leaveTangent=-1.8785, tangentMode=RCTM_User, time=0.5, value=0.7}]}
         RedCurve {keys [<at 0 0 RCIM_Cubic>,<at 0.5 0 RCIM_Cubic>]}
        }}))
     1=&MakeVectorFromLinearColorRGB(Color=&ColorFromCurve(ColorCurve=DataInterface() {propertyValues {
         AlphaCurve {keys [<at 0 1 RCIM_Cubic>]}
         BlueCurve {keys [<at 0 0 RCIM_Cubic>, {arriveTangent=2.5158, interpMode=RCIM_Cubic, leaveTangent=2.5158, tangentMode=RCTM_User, time=0.5, value=0}]}
         GreenCurve {keys [
           <at 0 0.5833 RCIM_Cubic>
           {arriveTangent=-1.8785, interpMode=RCIM_Cubic, leaveTangent=-1.8785, tangentMode=RCTM_User, time=0.5, value=0.4167}
          ]}
         RedCurve {keys [<at 0 1 RCIM_Cubic>,<at 0.5 1 RCIM_Cubic>]}
        }}))
     2=?
     3=?
     Index=$User.ColorVariant
     Num=::NewEnumerator0
    )
    ScaleA=true
    ScaleRGB=true
   )
   &ScaleSpriteSize(Scale Factor=&Vector2DFromFloat(Value=&FloatFromCurve(
      FloatCurve=DataInterface() {propertyValues {Curve {keys [
          {arriveTangent=1.3222, interpMode=RCIM_Cubic, leaveTangent=1.3222, tangentMode=RCTM_User, time=0, value=0}
          {arriveTangent=0.5373, interpMode=RCIM_Cubic, leaveTangent=0.5372, tangentMode=RCTM_User, time=1, value=1}
         ]}}}
      Scale Curve=3
     )))
   &DynamicMaterialParameters(
    Index 0 Param 1=&FloatFromCurve(FloatCurve=DataInterface() {propertyValues {Curve {keys [<at 0 1 RCIM_Cubic>, {arriveTangent=-2.2927, interpMode=RCIM_Cubic, leaveTangent=-2.2927, tangentMode=RCTM_User, time=1, value=0}]}}})
    Index 0 Param 2=1
    Index 0 Param 3=1
    Index 0 Param 4=1
    Module.Param0WriteEnabled=true
    Module.Param1WriteEnabled=true
    Module.Param2WriteEnabled=true
    Module.Param3WriteEnabled=true
   )
   &CameraOffset(Camera Offset Amount=5)
   &ScaleColor(
    Scale Alpha=&FN_PickFloatDependOnPlatform(
     TargetPlatform=DataInterface() {propertyValues {Platforms {qualityLevelMask=-32, deviceProfileStates [{profileName=Windows, qualityLevelMask=24, setQualityLevelMask=24}]}}}
     TargetPlatformActive=0.15
     TargetPlatformNoActive=1
    )
    ScaleA=true
    ScaleRGB=true
   ) {module=ScaleColor001}
  ]
  rendererClasses [&NiagaraSpriteRendererProperties]
  renderers [{rendererClass=&NiagaraSpriteRendererProperties}]
 }
 SpriteSplashesMouth {
  emitterUpdateScript [
   &EmitterState(
    Enable Distance Culling=true
    Life Cycle Mode=::NewEnumerator1
    Loop Duration=1
    MaxDistance=10000
    Module.bUseMaxDistance=true
    Module.bUseMinDistance=false
    Scalability Mode=::NewEnumerator1
    UseLoopDelay=false
   )
   &SpawnRate(
    SpawnRate=&Multiply_Float(
     A=&Sine(Normalized Angle=$Emitter.Age, Period=0.1, Scale=48, Bias=24)
     B=&FN_PickFloatDependOnPlatform(
      TargetPlatform=DataInterface() {propertyValues {Platforms {qualityLevelMask=-32, deviceProfileStates [{profileName=Windows, qualityLevelMask=24, setQualityLevelMask=24}]}}}
      TargetPlatformActive=0.5
      TargetPlatformNoActive=1
     )
    )
    Use Spawn Probability=false
   )
  ]
  particleSpawnScript [
   &InitializeParticle(
    Color=&MakeLinearColorFromVectorAndFloat(
     "Vector (RGB)"=<10 10 10>
     "Float (Alpha)"=&Lerp_Float(
      A=0.2
      B=1
      Alpha=&FloatFromCurve(
       FloatCurve=DataInterface() {propertyValues {Curve {keys [
           {interpMode=RCIM_Cubic, tangentMode=RCTM_User, time=0, value=0}
           {arriveTangent=0.8887, interpMode=RCIM_Cubic, leaveTangent=0.8887, tangentMode=RCTM_User, time=0.8, value=0.2}
           {arriveTangent=6.4377, interpMode=RCIM_Cubic, leaveTangent=6.4377, tangentMode=RCTM_User, time=1, value=1}
          ]}}}
       CurveIndex=&RandomRangeFloat(Minimum=0, Maximum=1)
      )
     )
    )
    Lifetime=&UniformRangedFloat(Minimum=0.2, Maximum=0.5)
    Module.Write Color=true
    Module.Write Lifetime=true
    Module.Write Mass=true
    Module.Write Material Random=true
    Module.Write Position=false
    Module.Write Scale=false
    Module.Write SpriteRotation=true
    Module.Write SpriteSize=true
    Position=&SimulationPosition()
    Sprite Rotation=&UniformRangedFloat(Minimum=0, Maximum=360)
    Sprite Size=&UniformRangedVector2D(Minimum=<30 30>, Maximum=<50 50>)
   )
   &SphereLocation(
    MaskSpawnGroup=false
    Override Seed=false
    Sphere Radius=2
    Surface Only=false
    UseNonUniformScale=false
    UseOffset=false
    UseSphereOrientation=false
   )
   &AddVelocity(Velocity=&Multiply_Vector(
     A=$Output.SphereLocation.SphereNormal
     B=&VectorFromFloat(Value=&UniformRangedFloat(Minimum=50, Maximum=100))
    )) {module=AddVelocity001}
   /SetVariables(Particles.SubImageIndex=&UniformRangedFloat(Minimum=0, Maximum=3)) {isSetParameters=true, module=SetVariables_#3}
   &AddVelocity(Velocity=&UniformRangedVector(Minimum=<-5 -5 50>, Maximum=<5 5 100>))
   &SystemLocation(MaskSpawnGroup=false, Offset=<0 0 41>)
  ]
  particleUpdateScript [
   &ParticleState()
   &AccelerationForce(Acceleration=<0 0 -980>)
   &Drag(
    Drag=&FloatFromCurve(FloatCurve=DataInterface() {propertyValues {Curve {keys [<at 0 0.5>,<at 0.8 4>]}}})
    Use Linear Drag=true
    Use Rotational Drag=false
   )
   &SpriteRotationRate(Rotation Rate=&Multiply_Float(
     A=&Lerp_Float(A=-1, B=1, Alpha=$Particles.Initial.MaterialRandom)
     B=&Lerp_Float(A=125, B=0, Alpha=&Power(a=$Particles.NormalizedAge, b=5))
    ))
   &SolveForcesAndVelocity(Clamp Velocity=false, Limit Acceleration=false)
   &ScaleColor(
    Scale RGB=&MergeHelper_VectorSwitch(
     0=&MakeVectorFromLinearColorRGB(Color=&ColorFromCurve(ColorCurve=DataInterface() {propertyValues {
         AlphaCurve {keys [<at 0 1 RCIM_Cubic>]}
         BlueCurve {keys [<at 0 0.3283 RCIM_Cubic>, {arriveTangent=2.5158, interpMode=RCIM_Cubic, leaveTangent=2.5158, tangentMode=RCTM_User, time=0.5, value=1}]}
         GreenCurve {keys [<at 0 1 RCIM_Cubic>, {arriveTangent=-1.8785, interpMode=RCIM_Cubic, leaveTangent=-1.8785, tangentMode=RCTM_User, time=0.5, value=0.7}]}
         RedCurve {keys [<at 0 0 RCIM_Cubic>,<at 0.5 0 RCIM_Cubic>]}
        }}))
     1=&MakeVectorFromLinearColorRGB(Color=&ColorFromCurve(ColorCurve=DataInterface() {propertyValues {
         AlphaCurve {keys [<at 0 1 RCIM_Cubic>]}
         BlueCurve {keys [<at 0 0 RCIM_Cubic>, {arriveTangent=2.5158, interpMode=RCIM_Cubic, leaveTangent=2.5158, tangentMode=RCTM_User, time=0.5, value=0}]}
         GreenCurve {keys [
           <at 0 0.5833 RCIM_Cubic>
           {arriveTangent=-1.8785, interpMode=RCIM_Cubic, leaveTangent=-1.8785, tangentMode=RCTM_User, time=0.5, value=0.4167}
          ]}
         RedCurve {keys [<at 0 1 RCIM_Cubic>,<at 0.5 1 RCIM_Cubic>]}
        }}))
     2=?
     3=?
     Index=$User.ColorVariant
     Num=::NewEnumerator0
    )
    ScaleA=true
    ScaleRGB=true
   )
   &ScaleSpriteSize(Scale Factor=&Vector2DFromFloat(Value=&FloatFromCurve(
      FloatCurve=DataInterface() {propertyValues {Curve {keys [
          {arriveTangent=1.3222, interpMode=RCIM_Cubic, leaveTangent=1.3222, tangentMode=RCTM_User, time=0, value=0.2}
          {arriveTangentWeight=0.3784, interpMode=RCIM_Cubic, tangentMode=RCTM_User, time=1, value=1}
         ]}}}
      Scale Curve=2
     )))
   &DynamicMaterialParameters(
    Index 0 Param 1=&FloatFromCurve(FloatCurve=DataInterface() {propertyValues {Curve {keys [<at 0 1 RCIM_Cubic>, {arriveTangent=-2.2927, interpMode=RCIM_Cubic, leaveTangent=-2.2927, tangentMode=RCTM_User, time=1, value=0}]}}})
    Index 0 Param 2=1
    Index 0 Param 3=1
    Index 0 Param 4=1
    Module.Param0WriteEnabled=true
    Module.Param1WriteEnabled=true
    Module.Param2WriteEnabled=true
    Module.Param3WriteEnabled=true
   )
   &CameraOffset(Camera Offset Amount=5)
   &ScaleColor(
    Scale Alpha=&FN_PickFloatDependOnPlatform(
     TargetPlatform=DataInterface() {propertyValues {Platforms {qualityLevelMask=-32, deviceProfileStates [{profileName=Windows, qualityLevelMask=24, setQualityLevelMask=24}]}}}
     TargetPlatformActive=0.45
     TargetPlatformNoActive=1
    )
    ScaleA=true
    ScaleRGB=true
   ) {module=ScaleColor001}
  ]
  rendererClasses [&NiagaraSpriteRendererProperties]
  renderers [{rendererClass=&NiagaraSpriteRendererProperties}]
 }
}
summary {
 systemName=NS_Consumable_ChugJug
 emitters [
  {emitterName=SpriteSplashesBig, rendererClasses [&NiagaraSpriteRendererProperties]}
  {emitterName=DripsGreen, rendererClasses [&NiagaraSpriteRendererProperties]}
  {emitterName=DripsBlue, rendererClasses [&NiagaraSpriteRendererProperties]}
  {emitterName=Source, simTarget=GPUComputeSim, rendererClasses [&NiagaraSpriteRendererProperties]}
  {emitterName=SpriteSplashesMouth, rendererClasses [&NiagaraSpriteRendererProperties]}
  {
   emitterName=FN_PIC_FLIP_Fluid_ReadOtherEmitter_V1Behavior
   simTarget=GPUComputeSim
   rendererClasses [&NiagaraSpriteRendererProperties]
  }
 ]
 userVariables [User.ColorVariant {type=NiagaraInt32}]
}
