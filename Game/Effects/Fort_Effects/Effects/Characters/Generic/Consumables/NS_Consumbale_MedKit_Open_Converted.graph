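# Niagara effect for opening a med kit consumable, apparently converted from Cascade, producing a soft green glow.
# LightPre bursts a particle with light, component and sprite renderers, using an intensity curve that drives light color scale and a separately multiplied Gen5 intensity.
# Lightbeams continuously spawns short-lived, randomly sized, slowly rising SubUV sprites.
# RoundFlarePre adds repeating flare sprites near the origin, and all three emitters fade through alpha curves.
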NiagaraSystem /Game/Effects/Fort_Effects/Effects/Characters/Generic/Consumables/NS_Consumbale_MedKit_Open_Converted
dependencies {usedDataInterfaces [&NiagaraDataInterfaceCurve], usedDynamicInputs [], usedModules [&SystemState]}
emitters {
 LightPre {
  emitterUpdateScript [
   &EmitterState(Inactive Response=::NewEnumerator1, Life Cycle Mode=::NewEnumerator1, Loop Behavior=::NewEnumerator1, Loop Duration=5.5, UseLoopDelay=true)
   &SpawnBurst_Instantaneous(Use Spawn Probability=false, UseLoopCountLimit=false)
  ]
  particleSpawnScript [
   &InitializeParticle(
    Lifetime=4
    Position=&ConvertVectorToPosition(Input Position=<0 0 25>, Conversion Operation:ENiagara_LWCConvertVecToPos=::NewEnumerator0)
    Position Mode=::NewEnumerator1
    Sprite Size=<64 64>
    Sprite Size Mode=::NewEnumerator4
    UsePositionOffset=false
    Write Material Random=true
   )
   &Color(Color=&MakeLinearColorFromVectorAndFloat("Vector (RGB)"=<0.35 1 0.5634>, "Float (Alpha)"=&ClampFloat(Float=0.2)))
   &SystemLocation(MaskSpawnGroup=false, Offset=<-20 0 40>)
  ]
  particleUpdateScript [
   /SetVariables(Particles.LightExponent=1.3) {isSetParameters=true}
   &ParticleState()
   &ScaleColor(
    Scale Alpha=&FloatFromCurve(FloatCurve=DataInterface() {propertyValues {Curve {keys [<at 0 0 RCIM_Cubic>,<at 0.5 1>,<at 1 0 RCIM_Cubic>]}}})
    Scale RGB=&VectorFromFloat(Value=&FloatFromCurve(FloatCurve=DataInterface() {propertyValues {Curve {keys [<at 0 1 RCIM_Cubic>,<at 1 1 RCIM_Cubic>]}}}))
    ScaleA=true
    ScaleRGB=true
   )
   /SetVariables(Particles.LightIntensity:NiagaraFloat=&FloatFromCurve(
     FloatCurve=DataInterface() {propertyValues {Curve {keys [
         {arriveTangent=16.64, interpMode=RCIM_Cubic, leaveTangent=16.64, tangentMode=RCTM_User, time=0, value=0}
         {interpMode=RCIM_Cubic, tangentMode=RCTM_User, time=0.2, value=0.95}
         {arriveTangent=-1.1875, interpMode=RCIM_Cubic, leaveTangent=-1.1875, time=0.7763, value=0.85}
         {arriveTangent=-0.1016, interpMode=RCIM_Cubic, leaveTangent=-0.1016, tangentMode=RCTM_User, time=1, value=0}
        ]}}}
     Scale Curve=2
    )) {isSetParameters=true}
   /SetVariables(Particles.Gen5LightIntensity:NiagaraFloat=&Multiply_Float(A=$Particles.LightIntensity, B=500)) {isSetParameters=true}
   &CascadeConversion_LightProperties(LightBrightnessScale=5, LightColorScale=<1 1 1>, LightRadiusScale=1) {enabled=false}
   &FN_Light_Attributes(Color=<rgba 0 1.5 0.5 1>, Create Color Attribute=true, Scale Color=$Particles.LightIntensity)
  ]
  rendererClasses [
   &NiagaraLightRendererProperties
   &NiagaraComponentRendererProperties
   &NiagaraSpriteRendererProperties
  ]
  renderers [
   {rendererClass=&NiagaraLightRendererProperties, rendererIndex=0}
   {rendererClass=&NiagaraComponentRendererProperties, rendererIndex=1}
   {rendererClass=&NiagaraSpriteRendererProperties, rendererIndex=2}
  ]
 }
 Lightbeams {
  emitterUpdateScript [
   &EmitterState(Inactive Response=::NewEnumerator1, Life Cycle Mode=::NewEnumerator1, Loop Behavior=::NewEnumerator1, Loop Duration=4, UseLoopDelay=true)
   &SpawnRate(SpawnRate=8, Use Spawn Probability=false)
  ]
  particleSpawnScript [
   &InitializeParticle(
    Lifetime=&RandomRangeFloat(Minimum=0.75, Maximum=1.25)
    Sprite Size=&RandomRangeVector2D(Minimum=<100 50>, Maximum=<150 90>)
    Sprite Size Mode=::NewEnumerator4
    UsePositionOffset=false
    Write Material Random=true
   )
   &Color(Color=&MakeLinearColorFromVectorAndFloat("Vector (RGB)"=<0.35 1 0.5634>, "Float (Alpha)"=&ClampFloat(Float=0.275)))
   &AddVelocity(Velocity=<0 0 1>)
   &CameraOffset(Camera Offset Amount=10)
   /Niagara/Modules/Update/SubUV/SubUVAnimation(Number Of Frames=3, SubUV Animation Mode=::NewEnumerator1)
  ]
  particleUpdateScript [
   &ParticleState()
   &ScaleColor(
    Scale Alpha=&FloatFromCurve(FloatCurve=DataInterface() {propertyValues {Curve {keys [
         {interpMode=RCIM_Cubic, tangentMode=RCTM_User, time=0, value=0}
         {interpMode=RCIM_Cubic, tangentMode=RCTM_User, time=0.5, value=1}
         {tangentMode=RCTM_User, time=1, value=0}
        ]}}})
    ScaleA=true
    ScaleRGB=true
   )
   &SolveForcesAndVelocity(Clamp Velocity=false, Limit Acceleration=false)
  ]
  rendererClasses [&NiagaraSpriteRendererProperties]
  renderers [{rendererClass=&NiagaraSpriteRendererProperties}]
 }
 RoundFlarePre {
  emitterUpdateScript [
   &EmitterState(Inactive Response=::NewEnumerator1, Life Cycle Mode=::NewEnumerator1, Loop Behavior=::NewEnumerator1, Loop Duration=5.5, UseLoopDelay=true)
   &SpawnRate(SpawnRate=2, Use Spawn Probability=false)
  ]
  particleSpawnScript [
   &InitializeParticle(
    Position=&ConvertVectorToPosition(Input Position=<0 0 8>, Conversion Operation:ENiagara_LWCConvertVecToPos=::NewEnumerator0)
    Position Mode=::NewEnumerator1
    Sprite Size=<45 45>
    Sprite Size Mode=::NewEnumerator4
    UsePositionOffset=false
    Write Material Random=true
   )
   &Color(Color=&MakeLinearColorFromVectorAndFloat("Vector (RGB)"=<0.35 1 0.5634>, "Float (Alpha)"=&ClampFloat(Float=0.2)))
  ]
  particleUpdateScript [
   &ParticleState()
   &ScaleColor(
    Scale Alpha=&FloatFromCurve(FloatCurve=DataInterface() {propertyValues {Curve {keys [<at 0 0>,<at 0.5 1 RCIM_Cubic>,<at 1 0 RCIM_Cubic>]}}})
    ScaleA=true
    ScaleRGB=true
   )
   &DynamicMaterialParameters(
    Index 0 Param 1=2
    Index 0 Param 2=2
    Index 0 Param 3=1
    Index 0 Param 4=1
    Module.Index0Param0Write=true
    Module.Index0Param1Write=true
    Module.Index0Param2Write=true
    Module.Index0Param3Write=true
   )
   &CameraOffset(Camera Offset Amount=10)
  ]
  rendererClasses [&NiagaraSpriteRendererProperties]
  renderers [{rendererClass=&NiagaraSpriteRendererProperties}]
 }
}
summary {
 systemName=NS_Consumbale_MedKit_Open_Converted
 emitters [
  {emitterName=RoundFlarePre, rendererClasses [&NiagaraSpriteRendererProperties]}
  {
   emitterName=LightPre
   rendererClasses [
    &NiagaraLightRendererProperties
    &NiagaraComponentRendererProperties
    &NiagaraSpriteRendererProperties
   ]
  }
  {emitterName=Lightbeams, rendererClasses [&NiagaraSpriteRendererProperties]}
 ]
 userVariables []
}
